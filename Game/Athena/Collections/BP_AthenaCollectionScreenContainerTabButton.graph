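# Tab button widget for the Athena collection screen, with an icon, selection background and inherited new-item badge.
# Hover and selection animations scale the icon; selection also fades in the background and drives the icon material's Selected parameter.
# Construct enlarges the icon on mobile and enables the press sound after a short delay.
# ResetTabButton clears selection and restores idle visuals, using flags to suppress animation and mute sound.
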WidgetBlueprint /Game/Athena/Collections/BP_AthenaCollectionScreenContainerTabButton
animations [
 Selected {
  binding_count=2
  length_seconds=0.15
  object=&self:Selected
  sequence {bindings [
    Image_61 {
     id=#1
     display_name=LazyImage_Icon
     possessed_class=Image
     tracks [
      {
       type=MovieScene2DTransformTrack
       sections [{
         type=MovieScene2DTransformSection
         channels [
          Translation.X {type=MovieSceneScriptingFloatChannel}
          Translation.Y {type=MovieSceneScriptingFloatChannel}
          Angle {type=MovieSceneScriptingFloatChannel}
          Scale.X {type=MovieSceneScriptingFloatChannel, keys [[0,0.9],[3,1]]}
          Scale.Y {type=MovieSceneScriptingFloatChannel, keys [[0,0.9],[3,1]]}
          Shear.X {type=MovieSceneScriptingFloatChannel}
          Shear.Y {type=MovieSceneScriptingFloatChannel}
         ]
        }]
      }
      {
       type=MovieSceneWidgetMaterialTrack
       sections [{type=MovieSceneParameterSection, channels [Selected {type=MovieSceneScriptingFloatChannel, keys [[0,0],[3,1]]}]}]
      }
     ]
    }
    Background {
     id=#2
     possessed_class=Image
     tracks [
      {
       type=MovieSceneFloatTrack
       display_name=Render Opacity
       sections [{type=MovieSceneFloatSection, channels [None {type=MovieSceneScriptingFloatChannel, keys [[0,0],[3,1]]}]}]
      }
      {
       type=MovieScene2DTransformTrack
       sections [{
         type=MovieScene2DTransformSection
         channels [
          Translation.X {type=MovieSceneScriptingFloatChannel}
          Translation.Y {type=MovieSceneScriptingFloatChannel}
          Angle {type=MovieSceneScriptingFloatChannel}
          Scale.X {type=MovieSceneScriptingFloatChannel, keys [[0,0.8],[3,1]]}
          Scale.Y {type=MovieSceneScriptingFloatChannel, keys [[0,0.8],[3,1]]}
          Shear.X {type=MovieSceneScriptingFloatChannel}
          Shear.Y {type=MovieSceneScriptingFloatChannel}
         ]
        }]
      }
     ]
    }
   ]}
 }
 Hovered {
  binding_count=1
  length_seconds=0.15
  object=&self:Hovered
  sequence {bindings [Image_61 {
     id=#1
     display_name=LazyImage_Icon
     possessed_class=Image
     tracks [{
       type=MovieScene2DTransformTrack
       sections [{
         type=MovieScene2DTransformSection
         channels [
          Translation.X {type=MovieSceneScriptingFloatChannel}
          Translation.Y {type=MovieSceneScriptingFloatChannel}
          Angle {type=MovieSceneScriptingFloatChannel}
          Scale.X {type=MovieSceneScriptingFloatChannel, keys [[0,0.9],[3,0.95]]}
          Scale.Y {type=MovieSceneScriptingFloatChannel, keys [[0,0.9],[3,0.95]]}
          Shear.X {type=MovieSceneScriptingFloatChannel}
          Shear.Y {type=MovieSceneScriptingFloatChannel}
         ]
        }]
      }]
    }]}
 }
]
graphs [EventGraph {nodes [
   n1 = EnablePlaySound() {is=title K2Node_CustomEvent}
   > Set bPlaySound(bPlaySound:bool=true) -> (Output_Get:bool)
   Event Construct()
   n2 = IsMobileGame()
   > SetWidthOverride(
    self=Get SizeBox_mobile() -> (SizeBox_mobile:Size Box Object Reference)
    InWidthOverride=Select(Option 0:float=64, Option 1:float=90, Index=@n2) -> (ReturnValue:float)
   )
   > SetHeightOverride(self=Get SizeBox_mobile(), InHeightOverride=Select(Option 0=64, Option 1=90, Index=@n2))
   > Set Timer by Event(Delegate=@n1, Time=0.1, bLooping=false)
   Event On Hovered()
   n3 = GetSelected!(self:Common Button Base Object Reference) -> (ReturnValue:bool)
   > n4 = Branch(Condition=@n3)
   n5 = Get Hovered() -> (Hovered:Widget Animation Object Reference)
   PlayAnimationForward(execute=@n4.else, InAnimation=@n5)
   Event On Unhovered()
   > n6 = Branch(Condition=@n3)
   PlayAnimationReverse(execute=@n6.else, InAnimation=@n5)
   Event On Selected()
   n7 = Get Selected() -> (Selected:Widget Animation Object Reference)
   > StopAnimation(InAnimation=@n7)
   n8 = Get ResetInProcess() -> (ResetInProcess:bool)
   > n9 = Branch(Condition=@n8)
   n10 = Get Selected()
   PlayAnimationForward(execute=@n9.else, InAnimation=@n10)
   > Branch(Condition=Get bPlaySound() -> (bPlaySound:bool))
   > PlaySound2D(Sound=Get PressedSound() -> (PressedSound:Sound Base Object Reference))
   Event On Deselected()
   > StopAnimation(InAnimation=@n7)
   > n11 = Branch(Condition=@n8)
   PlayAnimationReverse(execute=@n11.else, InAnimation=@n10)
   Event ResetTabButton() {is=title K2Node_Event}
   > Set ResetInProcess(ResetInProcess:bool=true) -> (Output_Get:bool)
   > StopAnimation(InAnimation=Get Selected())
   > n12 = Sequence()
   > ClearSelection(self:Common Button Base Object Reference)
   > SetRenderOpacity(self=Get Background() -> (Background:Image Object Reference), InOpacity=0)
   > n13 = GetDynamicMaterial(self=Get LazyImage_Icon() -> (LazyImage_Icon:Common Lazy Image Object Reference))
   > SetScalarParameterValue(self=@n13, ParameterName=Selected, Value=0)
   > Set RenderTransform!(
    execute
    RenderTransform:Widget Transform Structure=Make Widget Transform!(Translation:Vector 2D Structure=(X=0,Y=0), Scale:Vector 2D Structure=(X=0.9,Y=0.9), Shear:Vector 2D Structure=(X=0,Y=0), Angle:float=0) -> (WidgetTransform:Widget Transform Structure) {is=title K2Node_MakeStruct}
    self:Widget Object Reference=Get LazyImage_Icon()
   ) -> (then, Output_Get:Widget Transform Structure)
   > Set bPlaySound(bPlaySound=false)
   > Set Timer by Event(Delegate=Create Event(), Time=0.1, bLooping=false)
   Set ResetInProcess(execute=@n12.then_1, ResetInProcess=false)
  ]}]
variables [
 bPlaySound {type=boolean}
 PressedSound {type=SoundBase, default=null}
 ResetInProcess {type=boolean}
]
widgets {
 info {
  inheritedWidgetCount=3
  parentClass=/Script/FortniteUI.AthenaCollectionScreenContainerTabButton
  rootWidgetClass=&Overlay
  widgetCount=6
 }
 widgets [
  Overlay_0 {widgetClassPath=&Overlay, widget=Overlay_0}
  Background {
   widgetClassPath=&Image
   bIsVariable=true
   parent=Overlay_0
   widget=Background
   properties {
    renderOpacity=0
    brush {
     bIsDynamicallyLoaded=false
     drawAs=Box
     imageType=NoImage
     mirroring=NoMirror
     resourceName=None
     resourceObject=/Game/Athena/Collections/Art/M_UI_CategoryTab_Selected
     tiling=NoTile
     imageSize=<32 32>
     margin=<ltrb 0.5 0 0.5 0>
     outlineSettings {
      bUseBrushTransparency=false
      roundingType=HalfHeightRadius
      width=0
      color {colorUseRule=UseColor_Specified, specifiedColor=<rgba 0 0 0 0>}
      cornerRadii=<0 0 0 0>
     }
     tintColor {colorUseRule=UseColor_Specified, specifiedColor=<rgba 1 1 1 1>}
     uVRegion {bIsValid=false, max=<0 0>, min=<0 0>}
    }
   }
   slot {class=OverlaySlot, properties {horizontalAlignment=HAlign_Fill, verticalAlignment=VAlign_Fill, padding=<ltrb -4 -2 -4 -2>}}
  }
  SizeBox_mobile {
   widgetClassPath=&SizeBox
   bIsVariable=true
   parent=Overlay_0
   widget=SizeBox_mobile
   properties {bOverride_HeightOverride=true, bOverride_WidthOverride=true, heightOverride=64, widthOverride=64}
   slot {class=OverlaySlot, properties {horizontalAlignment=HAlign_Center, verticalAlignment=VAlign_Center}}
  }
  LazyImage_Icon {
   widgetClassPath=&FortLazyImage
   bInherited=true
   bIsVariable=true
   parent=SizeBox_mobile
   widget=LazyImage_Icon
   properties {
    brush {
     bIsDynamicallyLoaded=false
     drawAs=Image
     imageType=NoImage
     mirroring=NoMirror
     resourceName=None
     resourceObject=/Game/Athena/Collections/Art/Fishing/M_UI_CategoryTab_ButtonIcon_Inst
     tiling=NoTile
     imageSize=<64 64>
     margin=<ltrb 0 0 0 0>
     outlineSettings {
      bUseBrushTransparency=false
      roundingType=HalfHeightRadius
      width=0
      color {colorUseRule=UseColor_Specified, specifiedColor=<rgba 0 0 0 0>}
      cornerRadii=<0 0 0 0>
     }
     tintColor {colorUseRule=UseColor_Specified, specifiedColor=<rgba 1 1 1 1>}
     uVRegion {bIsValid=false, max=<0 0>, min=<0 0>}
    }
    renderTransform {angle=0, scale=<0.9 0.9>, shear=<0 0>, translation=<0 0>}
   }
   slot {class=SizeBoxSlot}
  }
  NewBang {
   widgetClassPath=&CommonBorder
   bInherited=true
   bIsVariable=true
   parent=Overlay_0
   widget=NewBang
   properties {horizontalAlignment=HAlign_Center, renderOpacity=0, visibility=Collapsed, padding=<ltrb 7 6 7 6>}
   slot {class=OverlaySlot, properties {horizontalAlignment=HAlign_Right, padding=<ltrb 0 -12 4 0>}}
  }
  InputActionWidget {widgetClassPath=&CommonActionWidget, bInherited=true, widget=null}
 ]
}
