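# Leaderboard row widget for the fishing collection, built on AthenaCollectionLeaderboardEntryFishing, showing placement, name, score and avatar.
# On value set, nested visibility switchers show "Not Placed" for negative values, "???" as the result for zero, and the icon and value for positive ones.
# Initialization colors the name white for the local player or light blue otherwise; reset turns it darker blue.
# The two background color variables are unused in the graph.
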WidgetBlueprint /Game/Athena/Collections/BP_AthenaCollectionLeaderboardEntryFishing
graphs [EventGraph {nodes [
   n1 = Event On Value Set!() -> (OutputDelegate, then, Value:float) {is=title K2Node_Event}
   > SetActiveWidgetIndex(
    self:Common Visibility Switcher Object Reference=Get VisibilitySwitcher_PlacedNotPlaced() -> (VisibilitySwitcher_PlacedNotPlaced:Common Visibility Switcher Object Reference)
    Index=Select(Option 0:int=1, Option 1:int=0, Index="float >= float"(A=@n1.Value, B=0) {is=title K2Node_CallFunction}) -> (ReturnValue:int)
   )
   > SetActiveWidgetIndex(
    self=Get VisibilitySwitcher_Result() -> (VisibilitySwitcher_Result:Common Visibility Switcher Object Reference)
    Index=Select(Option 0=0, Option 1=1, Index="float > float"(A=@n1.Value, B=0) {is=title K2Node_CallFunction})
   )
   Event On Leaderboard Entry Reset() {is=title K2Node_Event}
   > SetColorAndOpacity(
    self=Get Text_DisplayName() -> (Text_DisplayName:Common Text Object Reference)
    InColorAndOpacity=Make Slate Color(SpecifiedColor=(R=0.004,G=0.1683,B=1,A=1))
   )
   n2 = Event On Leaderboard Entry Initialized!() -> (OutputDelegate, then, bIsLocalPlayer:bool) {is=title K2Node_Event}
   > SetColorAndOpacity(
    self=Get Text_DisplayName()
    InColorAndOpacity=Select(
     Option 0:Slate Color Structure=Make Slate Color(SpecifiedColor=(R=0.2159,G=0.6445,B=0.9823,A=1))
     Option 1:Slate Color Structure=Make Slate Color(SpecifiedColor=(R=1,G=1,B=1,A=1))
     Index=@n2.bIsLocalPlayer
    ) -> (ReturnValue:Slate Color Structure)
   )
  ]}]
variables [
 Bg_Color_Empty {type=LinearColor, default=<rgba 0 0 0.0152 0.4>}
 Bg_Color_Populated {type=LinearColor, default=<rgba 0 0.0908 1 0.4>}
]
widgets {
 info {
  inheritedWidgetCount=5
  parentClass=/Script/FortniteUI.AthenaCollectionLeaderboardEntryFishing
  rootWidgetClass=&SizeBox
  widgetCount=22
 }
 widgets [
  SizeBox_37 {widgetClassPath=&SizeBox, widget=SizeBox_37, properties {bOverride_HeightOverride=true, heightOverride=48}}
  Overlay_0 {widgetClassPath=&Overlay, parent=SizeBox_37, widget=Overlay_0, slot {class=SizeBoxSlot}}
  Image_bg {
   widgetClassPath=&Image
   bIsVariable=true
   parent=Overlay_0
   widget=Image_bg
   properties {
    visibility=SelfHitTestInvisible
    brush {
     bIsDynamicallyLoaded=false
     drawAs=Image
     imageType=NoImage
     mirroring=NoMirror
     resourceName=None
     resourceObject=null
     tiling=NoTile
     imageSize=<41.9809 33.1349>
     margin=<ltrb 0 0 0 0>
     outlineSettings {
      bUseBrushTransparency=false
      roundingType=HalfHeightRadius
      width=0
      color {colorUseRule=UseColor_Specified, specifiedColor=<rgba 0 0 0 0>}
      cornerRadii=<0 0 0 0>
     }
     tintColor {colorUseRule=UseColor_Specified, specifiedColor=<rgba 1 1 1 1>}
     uVRegion {bIsValid=false, max=<0 0>, min=<0 0>}
    }
    colorAndOpacity=<rgba 0 0.0908 1 0.2471>
   }
   slot {class=OverlaySlot, properties {horizontalAlignment=HAlign_Fill, verticalAlignment=VAlign_Fill}}
  }
  HorizontalBox_0 {
   widgetClassPath=&HorizontalBox
   parent=Overlay_0
   widget=HorizontalBox_0
   slot {class=OverlaySlot, properties {horizontalAlignment=HAlign_Fill, verticalAlignment=VAlign_Fill, padding=<ltrb 6 0 0 0>}}
  }
  VisibilitySwitcher_PlacedNotPlaced {
   widgetClassPath=&CommonVisibilitySwitcher
   bIsVariable=true
   parent=HorizontalBox_0
   widget=VisibilitySwitcher_PlacedNotPlaced
   slot {class=HorizontalBoxSlot, properties {size {sizeRule=Fill, value=1}}}
  }
  HorizontalBox_83 {
   widgetClassPath=&HorizontalBox
   parent=VisibilitySwitcher_PlacedNotPlaced
   widget=HorizontalBox_83
   slot {class=CommonVisibilitySwitcherSlot}
  }
  HorizontalBox_2 {
   widgetClassPath=&HorizontalBox
   parent=HorizontalBox_83
   widget=HorizontalBox_2
   slot {class=HorizontalBoxSlot, properties {verticalAlignment=VAlign_Center, padding=<ltrb 0 0 8 0>, size {sizeRule=Fill, value=1}}}
  }
  SizeBox_0 {
   widgetClassPath=&SizeBox
   parent=HorizontalBox_2
   widget=SizeBox_0
   properties {bOverride_MinDesiredWidth=true, minDesiredWidth=34.4847, widthOverride=1.1307}
   slot {class=HorizontalBoxSlot, properties {padding=<ltrb 0 0 8 0>}}
  }
  HorizontalBox_Placement {widgetClassPath=&HorizontalBox, bIsVariable=true, parent=SizeBox_0, widget=HorizontalBox_Placement, slot {class=SizeBoxSlot}}
  CommonTextBlock_76 {
   widgetClassPath=&CommonTextBlock
   parent=HorizontalBox_Placement
   widget=CommonTextBlock_76
   properties {
    text="##"
    colorAndOpacity {colorUseRule=UseColor_Specified, specifiedColor=<rgba 0 0.0908 1 0.4>}
    font {
     bForceMonospaced=false
     bMaterialIsStencil=false
     fontMaterial=null
     fontObject=&BurbankSmall
     letterSpacing=0
     monospacedWidth=1
     size=18
     skewAmount=0
     typefaceFontName=Black
     outlineSettings {
      bApplyOutlineToDropShadows=false
      bMiteredCorners=false
      bSeparateFillAlpha=false
      outlineMaterial=null
      outlineSize=0
      outlineColor=<rgba 0 0 0 1>
     }
    }
    margin=<ltrb 0 -16 0 0>
    renderTransform {angle=0, scale=<1 1>, shear=<-12 0>, translation=<0 0>}
    shadowOffset=<0 0>
   }
   slot {class=HorizontalBoxSlot, properties {verticalAlignment=VAlign_Center}}
  }
  Text_Placement {
   widgetClassPath=&CommonTextBlock
   bInherited=true
   parent=HorizontalBox_Placement
   widget=Text_Placement
   properties {
    text=1
    colorAndOpacity {colorUseRule=UseColor_Specified, specifiedColor=<rgba 0 0.0908 1 0.4>}
    font {
     bForceMonospaced=false
     bMaterialIsStencil=false
     fontMaterial=null
     fontObject=&BurbankSmall
     letterSpacing=0
     monospacedWidth=1
     size=24
     skewAmount=0
     typefaceFontName=Black
     outlineSettings {
      bApplyOutlineToDropShadows=false
      bMiteredCorners=false
      bSeparateFillAlpha=false
      outlineMaterial=null
      outlineSize=0
      outlineColor=<rgba 0 0 0 1>
     }
    }
    renderTransform {angle=0, scale=<1 1>, shear=<-12 0>, translation=<0 0>}
    shadowOffset=<0 0>
   }
   slot {class=HorizontalBoxSlot}
  }
  Text_DisplayName {
   widgetClassPath=&CommonTextBlock
   bInherited=true
   parent=HorizontalBox_2
   widget=Text_DisplayName
   properties {
    scrollStyle=&TextScrollStyle-Base_C
    colorAndOpacity {colorUseRule=UseColor_Specified, specifiedColor=<rgba 0.2159 0.6445 0.9823 1>}
    font {
     bForceMonospaced=false
     bMaterialIsStencil=false
     fontMaterial=null
     fontObject=&BurbankSmall
     letterSpacing=0
     monospacedWidth=1
     size=16
     skewAmount=0
     typefaceFontName=Black
     outlineSettings {
      bApplyOutlineToDropShadows=false
      bMiteredCorners=false
      bSeparateFillAlpha=false
      outlineMaterial=null
      outlineSize=0
      outlineColor=<rgba 0 0 0 1>
     }
    }
    renderTransform {angle=0, scale=<1 1>, shear=<-12 0>, translation=<0 0>}
    shadowOffset=<0 0>
   }
   slot {class=HorizontalBoxSlot, properties {verticalAlignment=VAlign_Center}}
  }
  VisibilitySwitcher_Result {
   widgetClassPath=&CommonVisibilitySwitcher
   bIsVariable=true
   parent=HorizontalBox_83
   widget=VisibilitySwitcher_Result
   properties {activeWidgetIndex=1}
   slot {class=HorizontalBoxSlot, properties {verticalAlignment=VAlign_Center, padding=<ltrb 8 0 8 0>}}
  }
  Text_MissingResult {
   widgetClassPath=&CommonTextBlock
   parent=VisibilitySwitcher_Result
   widget=Text_MissingResult
   properties {
    text=???
    colorAndOpacity {colorUseRule=UseColor_Specified, specifiedColor=<rgba 0.004 0.1683 1 1>}
    font {
     bForceMonospaced=false
     bMaterialIsStencil=false
     fontMaterial=null
     fontObject=&BurbankSmall
     letterSpacing=0
     monospacedWidth=1
     size=14
     skewAmount=0
     typefaceFontName=Black
     outlineSettings {
      bApplyOutlineToDropShadows=false
      bMiteredCorners=false
      bSeparateFillAlpha=false
      outlineMaterial=null
      outlineSize=0
      outlineColor=<rgba 0 0 0 1>
     }
    }
    renderTransform {angle=0, scale=<1 1>, shear=<-12 0>, translation=<0 0>}
    shadowOffset=<0 0>
   }
   slot {class=CommonVisibilitySwitcherSlot, properties {verticalAlignment=VAlign_Center}}
  }
  HorizontalBox_1 {
   widgetClassPath=&HorizontalBox
   parent=VisibilitySwitcher_Result
   widget=HorizontalBox_1
   slot {class=CommonVisibilitySwitcherSlot}
  }
  Image_Stars {
   widgetClassPath=&Image
   bInherited=true
   bIsVariable=true
   parent=HorizontalBox_1
   widget=Image_Stars
   properties {
    visibility=SelfHitTestInvisible
    brush {
     bIsDynamicallyLoaded=false
     drawAs=Image
     imageType=NoImage
     mirroring=NoMirror
     resourceName=None
     resourceObject=/Game/Athena/Collections/Art/M_Ui_Collection_HighScore_IconSmall
     tiling=NoTile
     imageSize=<32 32>
     margin=<ltrb 0 0 0 0>
     outlineSettings {
      bUseBrushTransparency=false
      roundingType=HalfHeightRadius
      width=0
      color {colorUseRule=UseColor_Specified, specifiedColor=<rgba 0 0 0 0>}
      cornerRadii=<0 0 0 0>
     }
     tintColor {colorUseRule=UseColor_Specified, specifiedColor=<rgba 1 1 1 1>}
     uVRegion {bIsValid=false, max=<0 0>, min=<0 0>}
    }
   }
   slot {class=HorizontalBoxSlot, properties {horizontalAlignment=HAlign_Left, verticalAlignment=VAlign_Center, padding=<ltrb 6 0 4 0>}}
  }
  RichText_Value {
   widgetClassPath=&CommonRichTextBlock
   bInherited=true
   parent=HorizontalBox_1
   widget=RichText_Value
   properties {
    bOverrideDefaultStyle=true
    textTransformPolicy=ToUpper
    defaultTextStyleOverride {
     overflowPolicy=Clip
     transformPolicy=None
     colorAndOpacity {colorUseRule=UseColor_Specified, specifiedColor=<rgba 1 1 1 1>}
     font {
      bForceMonospaced=false
      bMaterialIsStencil=false
      fontMaterial=/Game/Athena/Collections/Art/M_Ui_Collection_HighScore_Inst
      fontObject=&BurbankSmall
      letterSpacing=0
      monospacedWidth=1
      size=17
      skewAmount=0
      typefaceFontName=Black
      outlineSettings {
       bApplyOutlineToDropShadows=false
       bMiteredCorners=false
       bSeparateFillAlpha=false
       outlineMaterial=null
       outlineSize=0
       outlineColor=<rgba 0 0 0 1>
      }
     }
     highlightColor {colorUseRule=UseColor_Specified, specifiedColor=<rgba 0 0 0 1>}
     highlightShape {
      bIsDynamicallyLoaded=false
      drawAs=Image
      imageType=NoImage
      mirroring=NoMirror
      resourceName=None
      resourceObject=null
      tiling=NoTile
      imageSize=<32 32>
      margin=<ltrb 0 0 0 0>
      outlineSettings {
       bUseBrushTransparency=false
       roundingType=HalfHeightRadius
       width=0
       color {colorUseRule=UseColor_Specified, specifiedColor=<rgba 0 0 0 0>}
       cornerRadii=<0 0 0 0>
      }
      tintColor {colorUseRule=UseColor_Specified, specifiedColor=<rgba 1 1 1 1>}
      uVRegion {bIsValid=false, max=<0 0>, min=<0 0>}
     }
     selectedBackgroundColor {colorUseRule=UseColor_Foreground, specifiedColor=<rgba 1 0 1 1>}
     shadowColorAndOpacity=<rgba 0 0 0 1>
     shadowOffset=<0 0>
     strikeBrush {
      bIsDynamicallyLoaded=false
      drawAs=Image
      imageType=NoImage
      mirroring=NoMirror
      resourceName=None
      resourceObject=null
      tiling=NoTile
      imageSize=<32 32>
      margin=<ltrb 0 0 0 0>
      outlineSettings {
       bUseBrushTransparency=false
       roundingType=HalfHeightRadius
       width=0
       color {colorUseRule=UseColor_Specified, specifiedColor=<rgba 0 0 0 0>}
       cornerRadii=<0 0 0 0>
      }
      tintColor {colorUseRule=UseColor_Specified, specifiedColor=<rgba 1 1 1 1>}
      uVRegion {bIsValid=false, max=<0 0>, min=<0 0>}
     }
     underlineBrush {
      bIsDynamicallyLoaded=false
      drawAs=Image
      imageType=NoImage
      mirroring=NoMirror
      resourceName=None
      resourceObject=null
      tiling=NoTile
      imageSize=<32 32>
      margin=<ltrb 0 0 0 0>
      outlineSettings {
       bUseBrushTransparency=false
       roundingType=HalfHeightRadius
       width=0
       color {colorUseRule=UseColor_Specified, specifiedColor=<rgba 0 0 0 0>}
       cornerRadii=<0 0 0 0>
      }
      tintColor {colorUseRule=UseColor_Specified, specifiedColor=<rgba 1 1 1 1>}
      uVRegion {bIsValid=false, max=<0 0>, min=<0 0>}
     }
    }
    renderTransform {angle=0, scale=<1 1>, shear=<-12 0>, translation=<0 0>}
   }
   slot {class=HorizontalBoxSlot}
  }
  Text_NotPlaced {
   widgetClassPath=&CommonTextBlock
   parent=VisibilitySwitcher_PlacedNotPlaced
   widget=Text_NotPlaced
   properties {
    text=Not Placed
    textTransformPolicy=ToUpper
    colorAndOpacity {colorUseRule=UseColor_Specified, specifiedColor=<rgba 0.004 0.1683 1 1>}
    font {
     bForceMonospaced=false
     bMaterialIsStencil=false
     fontMaterial=null
     fontObject=&Burbank
     letterSpacing=0
     monospacedWidth=1
     size=16
     skewAmount=0
     typefaceFontName=Burbank-PlayerPower
     outlineSettings {
      bApplyOutlineToDropShadows=false
      bMiteredCorners=false
      bSeparateFillAlpha=false
      outlineMaterial=null
      outlineSize=0
      outlineColor=<rgba 0 0 0 1>
     }
    }
    shadowOffset=<0 0>
   }
   slot {class=CommonVisibilitySwitcherSlot, properties {verticalAlignment=VAlign_Center, padding=<ltrb 12 0 0 0>}}
  }
  ScaleBox_0 {
   widgetClassPath=&ScaleBox
   parent=HorizontalBox_0
   widget=ScaleBox_0
   properties {stretch=ScaleToFitY}
   slot {class=HorizontalBoxSlot, properties {horizontalAlignment=HAlign_Center}}
  }
  Overlay_2 {widgetClassPath=&Overlay, parent=ScaleBox_0, widget=Overlay_2, slot {class=ScaleBoxSlot}}
  Image_122 {
   widgetClassPath=&Image
   bIsVariable=true
   parent=Overlay_2
   widget=Image_122
   properties {brush {
     bIsDynamicallyLoaded=false
     drawAs=Image
     imageType=NoImage
     mirroring=NoMirror
     resourceName=None
     resourceObject=/Game/Athena/UI/Frontend/Art/M_UI_RadialGradientMaster_Inst
     tiling=NoTile
     imageSize=<32 32>
     margin=<ltrb 0 0 0 0>
     outlineSettings {
      bUseBrushTransparency=false
      roundingType=HalfHeightRadius
      width=0
      color {colorUseRule=UseColor_Specified, specifiedColor=<rgba 0 0 0 0>}
      cornerRadii=<0 0 0 0>
     }
     tintColor {colorUseRule=UseColor_Specified, specifiedColor=<rgba 1 1 1 1>}
     uVRegion {bIsValid=false, max=<0 0>, min=<0 0>}
    }}
   slot {class=OverlaySlot}
  }
  AvatarIcon {
   widgetClassPath=/Script/FortniteUI.FortSocialAvatarIcon
   bInherited=true
   parent=Overlay_2
   widget=AvatarIcon
   properties {
    avatarMaterial=/Game/UI/SocialPanel/Art/Materials/UI_UserPortrait_Face
    materialTextureParamName=SmallPreview
    brush {
     bIsDynamicallyLoaded=false
     drawAs=Image
     imageType=NoImage
     mirroring=NoMirror
     resourceName=None
     resourceObject=&UI_UserPortrait_Face
     tiling=NoTile
     imageSize=<32 32>
     margin=<ltrb 0 0 0 0>
     outlineSettings {
      bUseBrushTransparency=false
      roundingType=HalfHeightRadius
      width=0
      color {colorUseRule=UseColor_Specified, specifiedColor=<rgba 0 0 0 0>}
      cornerRadii=<0 0 0 0>
     }
     tintColor {colorUseRule=UseColor_Specified, specifiedColor=<rgba 1 1 1 1>}
     uVRegion {bIsValid=false, max=<0 0>, min=<0 0>}
    }
   }
   slot {class=OverlaySlot}
  }
 ]
}
